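# Control Rig for the large male FlavorStock Bat body that runs the shared DeformRig function, then adds corrective motion for the right wristband.
# A parent constraint blends the wristband between lowerarm_twist_01_r and hand_r before the local wrist_function runs.
# That function remaps hand_r Euler rotation from its initial pose into weights that offset the wristband bones, plus one constant downward offset.
# Unwired alternatives remain in the function, including attribute-driven offsets and another constraint.
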ControlRigBlueprint /BRCosmetics/Characters/Player/Male/Large/Bodies/M_LRG_FlavorStock_Bat/Meshes/M_LRG_FlavorStock_Bat_CtrlRig
graphs [
 New Function_ContainedGraph {nodes [
   RigVMFunctionEntryNode()
   n1 = &RigVMFunction_MathTransformMakeRelative(
    Global=&RigUnit_GetTransform(Item=(Type=Bone,Name="hand_r"), Space=LocalSpace).Transform
    Parent=&RigUnit_GetTransform(Item=(Type=Bone,Name="hand_r"), Space=LocalSpace, bInitial=true).Transform
   )
   n2 = &RigVMFunction_MathQuaternionToEuler(RotationOrder=ZYX, Value=@n1.Local.Rotation)
   n3 = &RigVMFunction_MathDoubleRemap(SourceMaximum=-45, SourceMinimum=-1, TargetMaximum=1, TargetMinimum=0, Value=@n2.Result.Y)
   > &RigUnit_ModifyTransforms(
    ItemToModify=((Item=(Type=Bone,Name="wristband_r"),Transform=(Rotation=(X=0,Y=0,Z=0,W=1),Translation=(X=0,Y=1,Z=0),Scale3D=(X=1,Y=1,Z=1))))
    WeightMinimum=0
    Weight=@n3.Result
   )
   > &RigUnit_ModifyTransforms(
    ItemToModify=((Item=(Type=Bone,Name="wristband"),Transform=(Rotation=(X=0,Y=0,Z=0,W=1),Translation=(X=-0.5,Y=-0.6949,Z=0),Scale3D=(X=1,Y=1,Z=1))))
    WeightMinimum=0
    Weight=@n3.Result
   )
   n4 = &RigVMFunction_MathDoubleRemap(SourceMaximum=45, SourceMinimum=1, TargetMaximum=1, TargetMinimum=0, Value=@n2.Result.Y)
   > &RigUnit_ModifyTransforms(
    ItemToModify=((Item=(Type=Bone,Name="wristband_l"),Transform=(Rotation=(X=0,Y=0,Z=0,W=1),Translation=(X=0,Y=-2,Z=0),Scale3D=(X=1,Y=1,Z=1))))
    WeightMinimum=0
    Weight=@n4.Result
   )
   > &RigUnit_ModifyTransforms(
    ItemToModify=((Item=(Type=Bone,Name="wristband"),Transform=(Rotation=(X=0,Y=0,Z=0,W=1),Translation=(X=-0.5,Y=0.5,Z=-0.1),Scale3D=(X=1,Y=1,Z=1))))
    WeightMinimum=0
    Weight=@n4.Result
   )
   n5 = &RigVMFunction_MathQuaternionToEuler(RotationOrder=XZY, Value=@n1.Local.Rotation)
   n6 = &RigVMFunction_MathDoubleRemap(SourceMaximum=-45, SourceMinimum=-1, TargetMaximum=1, TargetMinimum=0, Value=@n5.Result.Z)
   > &RigUnit_ModifyTransforms(
    ItemToModify=((Item=(Type=Bone,Name="wristband_up"),Transform=(Rotation=(X=0,Y=0,Z=0,W=1),Translation=(X=0,Y=0,Z=1),Scale3D=(X=1,Y=1,Z=1))))
    WeightMinimum=0
    Weight=@n6.Result
   )
   > &RigUnit_ModifyTransforms(
    ItemToModify=((Item=(Type=Bone,Name="wristband_down"),Transform=(Rotation=(X=0,Y=0,Z=0,W=1),Translation=(X=0,Y=0,Z=-2),Scale3D=(X=1,Y=1,Z=1))))
    WeightMinimum=0
    Weight=@n6.Result
   )
   > &RigUnit_ModifyTransforms(
    ItemToModify=((Item=(Type=Bone,Name="wristband_up"),Transform=(Rotation=(X=0,Y=0,Z=0,W=1),Translation=(X=0,Y=0,Z=2),Scale3D=(X=1,Y=1,Z=1))))
    WeightMinimum=0
    Weight=&RigVMFunction_MathDoubleRemap(SourceMaximum=45, SourceMinimum=1, TargetMaximum=1, TargetMinimum=0, Value=@n5.Result.Z).Result
   )
   > &RigUnit_ModifyTransforms(
    ItemToModify=((Item=(Type=Bone,Name="wristband_down"),Transform=(Rotation=(X=0,Y=0,Z=0,W=1),Translation=(X=0,Y=0,Z=-1),Scale3D=(X=1,Y=1,Z=1))))
    Weight=1
    WeightMinimum=0
   )
   > RigVMFunctionReturnNode()
   &RigUnit_ModifyTransforms(
    ExecutePin=()
    ItemToModify=((Item=(Type=Bone,Name="wristband_up"),Transform=(Rotation=(X=0,Y=0,Z=0,W=1),Translation=(X=0,Y=0,Z=100),Scale3D=(X=0.9,Y=1,Z=1))))
    WeightMinimum=0
    Weight=RigVMDispatchNode(BoneName=root, Default=0, Name=deformRig_r_wrist_up_alpha) {template="DISPATCH_RigDispatch_GetAnimAttribute(in Name,in BoneName,in Default,out Value,out Found)"}.Value
   )
   > &RigUnit_ModifyTransforms(
    ItemToModify=((Item=(Type=Bone,Name="wristband_down"),Transform=(Rotation=(X=0,Y=0,Z=0,W=1),Translation=(X=0,Y=0,Z=-100),Scale3D=(X=0.9,Y=1,Z=1))))
    WeightMinimum=0
    Weight=RigVMDispatchNode(BoneName=root, Default=0, Name=deformRig_r_wrist_dn_alpha) {template="DISPATCH_RigDispatch_GetAnimAttribute(in Name,in BoneName,in Default,out Value,out Found)"}.Value
   )
   n7 = &RigVMFunction_MathQuaternionToEuler(
    RotationOrder=ZYX
    Value=&RigVMFunction_MathTransformMakeRelative(
     Global=&RigUnit_GetTransform(Item=(Type=Bone,Name="hand_r"), Space=LocalSpace).Transform
     Parent=&RigUnit_GetTransform(Item=(Type=Bone,Name="hand_r"), Space=LocalSpace, bInitial=true).Transform
    ).Local.Rotation
   )
   &RigUnit_ModifyTransforms(
    ExecutePin=()
    ItemToModify=((Item=(Type=Bone,Name="wristband_down"),Transform=(Rotation=(X=0,Y=0,Z=0,W=1),Translation=(X=-1,Y=0.5,Z=0.5),Scale3D=(X=1,Y=1.1,Z=1.1))))
    WeightMinimum=0
    Weight=&RigVMFunction_MathDoubleRemap(SourceMaximum=-45, SourceMinimum=-1, TargetMaximum=1, TargetMinimum=0, Value=@n7.Result.Y).Result
   )
   > &RigUnit_ModifyTransforms(
    ItemToModify=((Item=(Type=Bone,Name="wristband"),Transform=(Rotation=(X=0,Y=0,Z=0,W=1),Translation=(X=-1,Y=0,Z=0.5),Scale3D=(X=1,Y=1.1,Z=1.1))))
    WeightMinimum=0
    Weight=&RigVMFunction_MathDoubleRemap(SourceMaximum=45, SourceMinimum=5, TargetMaximum=1, TargetMinimum=0, Value=@n7.Result.Y).Result
   )
   &RigUnit_ParentConstraint(
    AdvancedSettings=(InterpolationType=Shortest,RotationOrderForFilter=XZY)
    Child=(Type=Bone,Name="wristband")
    ExecutePin=()
    Parents=((Item=(Type=Bone,Name="lowerarm_r"),Weight=0.75),(Item=(Type=Bone,Name="hand_r"),Weight=0.25))
    Weight=1
   )
  ]}
 RigVMFunctionLibrary {nodes [wrist_function = RigVMCollapseNode(ExecuteContext=())]}
 RigVMModel {nodes [
   n1 = &RigUnit_PrepareForExecution()
   &RigUnit_HierarchyImportFromSkeleton(NameSpace=None, bIncludeCurves=False, bIncludeMeshSockets=False, bIncludeVirtualBones=True, ExecutePin=@n1.ExecutePin)
   RigUnit_BeginExecution = &RigUnit_BeginExecution()
   FN_DeformRig_FNC = RigVMFunctionReferenceNode(
    DeformRig_ Curve="(deformRig_l_upperarm_up_alpha_72_#1=False,deformRig_r_upperarm_up_alpha_73_#2=False,deformRig_l_upperarm_fr_alpha_74_#3=False,deformRig_r_upperarm_fr_alpha_75_CB51B8 ...(truncated)"
    DeformRig_Bool="(upperarm_1_#4=True,elbow_3_#5=True,wrist_7_#6=True,groin_8_#7=True,glute_17_#8=True,knee_9_40C408324CA261 ...(truncated)"
    DeformRig_Debug=(upperarm_debug_18_#4=False,wrist_debug_20_#6=False,lat_debug_26_#9=False,groinGlute_debug_27_#7=False)
    DeformRig_Value="(upperarm_up_18_#10=(Rotation=(X=0.0,Y=-0.0,Z=0.0,W=1.0),Translation=(X=0.0,Y=0.0,Z=8.0),Scale3D=(X=0.70,Y=1.0,Z=1.0)),upperarm_fwd_fr_19_#11=(Rotation=(X=0.0,Y=-0.0,Z=0.0,W=1.0),Translation=(X=0.0, ...(truncated)"
    ExecuteContext=@RigUnit_BeginExecution.ExecutePin
   )
   > &RigUnit_ParentConstraint(
    AdvancedSettings=(InterpolationType=Shortest,RotationOrderForFilter=XZY)
    Child=(Type=Bone,Name="wristband")
    Parents=((Item=(Type=Bone,Name="lowerarm_twist_01_r"),Weight=0.85),(Item=(Type=Bone,Name="hand_r"),Weight=0.15))
    Weight=1
   )
   > wrist_function = RigVMFunctionReferenceNode()
  ]}
]
